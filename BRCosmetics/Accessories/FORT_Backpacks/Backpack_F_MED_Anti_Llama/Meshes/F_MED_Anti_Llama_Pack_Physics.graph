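# Physics asset for the Anti Llama backpack cosmetic, giving its two swords simple collision bodies for physical motion relative to the pack.
# A small capsule represents the pack bone, and a longer, offset capsule covers each sword bone.
# Each sword is constrained to the pack with twist and swing 1 locked, leaving swing 2 as its hinge-like axis, with slightly different small limit angles per sword.
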PhysicsAsset /BRCosmetics/Accessories/FORT_Backpacks/Backpack_F_MED_Anti_Llama/Meshes/F_MED_Anti_Llama_Pack_Physics
bodies {
 pack [{shapeType=Capsule, length=1, radius=1}]
 swordA [{shapeType=Capsule, length=9.75, radius=2, center=<5.6565 0 0>, rotation=<rot 90 0 0>}]
 swordB [{shapeType=Capsule, length=9.75, radius=2, center=<5.6565 0 0>, rotation=<rot 90 0 0>}]
}
constraints [
 {bone1Name=swordA, bone2Name=pack, swing1LimitDegrees=45, swing1Motion=Locked, swing2LimitDegrees=2, twistLimitDegrees=45, twistMotion=Locked}
 {bone1Name=swordB, bone2Name=pack, swing1LimitDegrees=45, swing1Motion=Locked, swing2LimitDegrees=3, twistLimitDegrees=45, twistMotion=Locked}
]
